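# Sound cue for the rumble accompanying the camera shake of a temple puzzle door, part of the item-check puzzle assets.
# A fixed short delay feeds a modulator, which plays a single door cam-shake wave with slight variation.
# It plays as indoor ambient sound, spatialized through the puzzle door's quad attenuation settings at a reduced volume.
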SoundCue /ChronoEnviroItems/TemplePuzzles/ItemCheck/Assets/Audio/Amb_Door_Puzzle_CamShake_Cue
root=Delay_0
nodes [
 {id=Delay_0, children [Modulator_0], properties {delayMax=0.4, delayMin=0.4}}
 {id=Modulator_0, children [WavePlayer_1]}
 {id=WavePlayer_1, properties {soundWaveAssetPtr=&/S25_Puzzle_Door_CamShake}}
]
properties {
 attenuationSettings=&/Att_Puzzle_DoorQuad
 maxDistance=1500
 soundClassObject=&Fort_Ambient_Indoors_Athena
 volumeMultiplier=0.702
}
