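# Looping ambient sound cue for the door in a temple item-check puzzle, classed as indoor ambience.
# A single looping wave player feeds a default Modulator, with the cue's volume reduced to about half.
# Overridden attenuation adds distance falloff, stereo spread and trace-based occlusion with a low-pass filter.
# Overridden concurrency allows one instance at a time and prevents new ones from starting.
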SoundCue /ChronoEnviroItems/TemplePuzzles/ItemCheck/Assets/Audio/Amb_Door_Puzzle_Loop_Cue
root=Modulator_0
nodes [
 {id=Modulator_0, children [WavePlayer_0]}
 {id=WavePlayer_0, properties {bLooping=true, soundWaveAssetPtr=&/S25_Puzzle_Door_Loop}}
]
properties {
 bOverrideAttenuation=true
 bOverrideConcurrency=true
 maxDistance=1700
 soundClassObject=&Fort_Ambient_Indoors_Athena
 volumeMultiplier=0.501
 attenuationOverrides {
  bEnableOcclusion=true
  falloffDistance=1500
  occlusionInterpolationTime=2
  occlusionLowPassFilterFrequency=900
  occlusionTraceChannel=ECC_GameTraceChannel6
  stereoSpread=350
  attenuationShapeExtents {x=200}
 }
 concurrencyOverrides {maxCount=1, resolutionRule=PreventNew, retriggerTime=2}
}
