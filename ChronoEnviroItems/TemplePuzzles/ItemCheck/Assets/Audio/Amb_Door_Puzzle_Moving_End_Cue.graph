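# Sound cue for the stop sound of a moving temple puzzle door, playing a door-close wave through a modulator.
# It uses a puzzle door attenuation and the indoor ambient sound class at slightly reduced volume.
# A concurrency override allows only one instance at a time and blocks new ones rather than cutting off the playing sound, with a short retrigger time.
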SoundCue /ChronoEnviroItems/TemplePuzzles/ItemCheck/Assets/Audio/Amb_Door_Puzzle_Moving_End_Cue
root=Modulator_0
nodes [
 {id=Modulator_0, children [WavePlayer_0]}
 {id=WavePlayer_0, properties {soundWaveAssetPtr=&/S25_Puzzle_Door_Close_St}}
]
properties {
 attenuationSettings=&/Att_Puzzle_Door
 bOverrideConcurrency=true
 maxDistance=2650
 soundClassObject=&Fort_Ambient_Indoors_Athena
 volumeMultiplier=0.802
 concurrencyOverrides {maxCount=1, resolutionRule=PreventNew, retriggerTime=3}
}
