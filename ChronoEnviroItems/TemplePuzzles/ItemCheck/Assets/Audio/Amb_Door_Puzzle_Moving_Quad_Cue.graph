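# Ambient sound cue for a temple puzzle door moving, classed as indoor ambience.
# A quality-level root sends its first two levels to a mix of quad and stereo waves, and its last two to the stereo wave alone.
# The quad layer has its own ambient attenuation node, separate from the cue's door attenuation.
# A concurrency override caps simultaneous instances, stopping the oldest, with a retrigger time.
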SoundCue /ChronoEnviroItems/TemplePuzzles/ItemCheck/Assets/Audio/Amb_Door_Puzzle_Moving_Quad_Cue
root=QualityLevel_1
nodes [
 {id=Attenuation_2, children [WavePlayer_0], properties {attenuationSettings=&Fort_Ambient_Athena_Att}}
 {id=Mixer_0, children [Attenuation_2, WavePlayer_1], properties {inputVolume [1,1]}}
 {id=QualityLevel_1, children [Mixer_0,Mixer_0,WavePlayer_1,WavePlayer_1]}
 {id=WavePlayer_0, properties {soundWaveAssetPtr=&/S25_Puzzle_Door_Moving_Quad}}
 {id=WavePlayer_1, properties {soundWaveAssetPtr=&/S25_Puzzle_Door_Moving_St}}
]
properties {
 attenuationSettings=&/Att_Puzzle_DoorQuad
 bOverrideConcurrency=true
 maxDistance=10000
 soundClassObject=&Fort_Ambient_Indoors_Athena
 volumeMultiplier=0.952
 concurrencyOverrides {maxCount=2, resolutionRule=StopOldest, retriggerTime=3}
}
