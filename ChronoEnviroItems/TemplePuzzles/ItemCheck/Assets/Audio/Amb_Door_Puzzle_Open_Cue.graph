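# Sound cue for an ambient temple puzzle door opening, judging by its name and wave asset.
# It routes a single wave player through a modulator and plays with puzzle-door attenuation in the indoor ambient sound class.
# A concurrency override allows one instance at a time, blocking new plays with a retrigger delay, and the sound keeps playing when silent.
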SoundCue /ChronoEnviroItems/TemplePuzzles/ItemCheck/Assets/Audio/Amb_Door_Puzzle_Open_Cue
root=Modulator_0
nodes [
 {id=Modulator_0, children [WavePlayer_2]}
 {id=WavePlayer_2, properties {soundWaveAssetPtr=&/S25_Puzzle_Door_Open_St}}
]
properties {
 attenuationSettings=&/Att_Puzzle_Door
 bOverrideConcurrency=true
 maxDistance=2650
 soundClassObject=&Fort_Ambient_Indoors_Athena
 virtualizationMode=PlayWhenSilent
 volumeMultiplier=1.353
 concurrencyOverrides {maxCount=1, resolutionRule=PreventNew, retriggerTime=2}
}
