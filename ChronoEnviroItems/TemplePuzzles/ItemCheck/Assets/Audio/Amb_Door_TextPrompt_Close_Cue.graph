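# Sound cue for the door text prompt closing sound in the ChronoEnviroItems temple puzzle item check.
# A single wave player feeds the close sound through a modulator at the root.
# Playback uses the puzzle door attenuation, a reduced volume and the non-critical in-game UI sound class.
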SoundCue /ChronoEnviroItems/TemplePuzzles/ItemCheck/Assets/Audio/Amb_Door_TextPrompt_Close_Cue
root=Modulator_0
nodes [
 {id=Modulator_0, children [WavePlayer_1]}
 {id=WavePlayer_1, properties {soundWaveAssetPtr=&/S25_Puzzle_Door_TextPrompt_Close}}
]
properties {
 attenuationSettings=&/Att_Puzzle_Door
 maxDistance=2650
 soundClassObject=&Fort_UI_InGame_NonCritical
 volumeMultiplier=0.6
}
