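# Sound cue for the text prompt opening on a puzzle door, part of the TemplePuzzles ItemCheck assets.
# A random node picks one of three equally weighted wave variations, which a modulator varies in pitch and volume after a short fixed delay.
# The cue plays at lowered pitch and volume, uses puzzle door attenuation, and belongs to the non-critical in-game UI sound class.
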SoundCue /ChronoEnviroItems/TemplePuzzles/ItemCheck/Assets/Audio/Amb_Door_TextPrompt_Open_Cue
root=Delay_0
nodes [
 {id=Delay_0, children [Modulator_0], properties {delayMax=0.1, delayMin=0.1}}
 {id=Modulator_0, children [Random_0], properties {pitchMin=0.8, volumeMin=0.6}}
 {id=Random_0, children [WavePlayer_3,WavePlayer_4,WavePlayer_5], properties {weights [1,1,1]}}
 {id=WavePlayer_3, properties {soundWaveAssetPtr=&/S25_Puzzle_Door_TextPrompt_Open_01}}
 {id=WavePlayer_4, properties {soundWaveAssetPtr=&/S25_Puzzle_Door_TextPrompt_Open_02}}
 {id=WavePlayer_5, properties {soundWaveAssetPtr=&/S25_Puzzle_Door_TextPrompt_Open_03}}
]
properties {
 attenuationSettings=&/Att_Puzzle_Door
 maxDistance=2650
 pitchMultiplier=0.8
 soundClassObject=&Fort_UI_InGame_NonCritical
 volumeMultiplier=0.55
}
