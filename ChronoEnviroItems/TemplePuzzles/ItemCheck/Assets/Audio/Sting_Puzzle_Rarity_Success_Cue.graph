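# Sound cue for the success sting of a rarity puzzle in the temple puzzle item-check content.
# A modulator with a fixed reduced volume and unshifted pitch feeds a single wave player for the S25 rarity puzzle sting.
# The cue uses the indoor ambient Athena sound class and a very large max distance.
# Concurrency overrides cap simultaneous instances, stopping the oldest, and set a retrigger time.
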SoundCue /ChronoEnviroItems/TemplePuzzles/ItemCheck/Assets/Audio/Sting_Puzzle_Rarity_Success_Cue
root=Modulator_0
nodes [
 {id=Modulator_0, children [WavePlayer_4], properties {pitchMax=1, pitchMin=1, volumeMax=0.6, volumeMin=0.6}}
 {id=WavePlayer_4, properties {soundWaveAssetPtr=&/S25_Sting_Puzzle_Rarity}}
]
properties {
 maxDistance=2097152
 soundClassObject=&Fort_Ambient_Indoors_Athena
 volumeMultiplier=0.802
 concurrencyOverrides {maxCount=2, resolutionRule=StopOldest, retriggerTime=1}
}
